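# Additive, unlit, two-sided Niagara mesh-particle material for DiamondSkin weakpoint light rays.
# Two counter-panning line textures, a noise mask and a center glow form rays with noise-varied vertical falloff.
# That mask blends and scales rainbow-texture and particle color into emissive, with dynamic parameters randomizing tiling and timing.
# Particle alpha, a V gradient, quality-switched depth fade and optional near-camera fade form a separate fade.
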Material /Game/Effects/Fort_Effects/GameplayPulginsAssets/DiamondSkin/Materials/M_DiamondSkin_Weakpoint_LightRays
graph {
 properties {
  bUsedWithNanite=true
  bUsedWithNiagaraMeshParticles=true
  blendMode=BLEND_Additive
  pixelDepthOffsetMode=PDOM_Legacy
  shadingModel=MSM_Unlit
  twoSided=true
  usedShadingModels=Unlit
 }
 nodes [
  n1 = VectorParameter() {group=Color, parameterName=Color UV, defaultValue=<rgba 1 0.2 0 -0.5>}
  n2 = DynamicParameter() {parameterIndex=1, defaultValue=<rgba 0 0 1 0>, paramNames [UVTilingRandom, TimeOffsetRandom, WPOStrength, TrailingModeON]}
  n3 = Multiply(
   A=TextureCoordinate() {coordinateIndex=1}
   B=MaterialFunctionCall(X=@n1.R, Y=Add(A=@n2.UVTilingRandom, B=@n1.G)) {materialFunction=&MakeFloat2}.Result
  )
  n4 = Multiply(
   A=MaterialFunctionCall(X=@n1.B, Y=@n1.B) {materialFunction=&MakeFloat2}.Result
   B=Add(A=@n2.TimeOffsetRandom, B=Time())
  )
  n5 = TextureSampleParameter2D!(UVs=Add(A=Add(A=@n3, B=@n4)), Level, Apply View MipBias) {group=Color, mipValueMode=TMVM_MipLevel, parameterName=Color Tex, texture=&T_Tile_RainbowColor}
  n6 = ScalarParameter() {group=Color, parameterName=Color Tex Desaturation}
  n7 = Multiply(A=Multiply(A=Desaturation(None=@n5.RGB, Fraction=@n6))) {constB=5}
  n8 = ParticleColor()
  n9 = DynamicParameter() {defaultValue=<rgba 0 1 1 1>, paramNames [Random 0-1, Panner Speed, "U Tile ", Param4]}
  n10 = Multiply(A=TextureCoordinate(), B=@n9."U Tile ")
  n11 = ScalarParameter() {defaultValue=2, parameterName=Gradient Tile A}
  n12 = Multiply(A=@n9.Panner Speed, B=Add(A=@n9.Random 0-1, B=Time()))
  n13 = TextureObjectParameter() {parameterName="Lines Texture (RG)", texture=/Game/Effects/Fort_Effects/Textures/Outlander/T_InTheZone_LightRays_01}
  n14 = TextureSample(UVs=Panner(Coordinate=Multiply(A=@n10, B=@n11), Time=@n12) {speedX=0.0702}, Tex=@n13) {texture=&T_InTheZone_LightRays_01}
  n15 = Multiply(A=@n9."U Tile ", B=TextureCoordinate())
  n16 = ScalarParameter() {defaultValue=2.5, parameterName=Gradient Tile B}
  n17 = TextureSample(UVs=Panner(Coordinate=Multiply(A=@n15, B=@n16), Time=@n12) {speedX=-0.047}, Tex=@n13) {texture=&T_InTheZone_LightRays_01}
  n18 = TextureCoordinate() {uTiling=0.33, vTiling=0.01}
  n19 = DynamicParameter() {defaultValue=<rgba 0 1 1 1>, paramNames [Random 0-1, Panner Speed, "U Tile ", Param4]}
  n20 = TextureSample(
   UVs=Panner(Coordinate=Add(A=@n18, B=@n19.Random 0-1), Time=@n12) {speedX=0.0436}
   Tex=TextureObjectParameter() {parameterName="Lines Texture Mask (R)", texture=&T_NoisePack_01}
  ) {texture=&T_NoisePack_01}
  n21 = ScalarParameter() {defaultValue=2, parameterName=Mask Brightness}
  n22 = TextureCoordinate()
  n23 = ScalarParameter() {defaultValue=-0.025, parameterName=Slide}
  n24 = OneMinus(None=ComponentMask(None=Add(A=@n22, B=@n23)) {g=true})
  n25 = ScalarParameter() {defaultValue=20, parameterName=Center Exponent}
  n26 = Add(
   A=Multiply(A=Multiply(A=@n14.R, B=@n17.G), B=Multiply(A=@n20.R, B=@n21))
   B=Multiply(A=Power(Base=@n24, Exp=@n25), B=ScalarParameter() {defaultValue=0.5, parameterName=Center Brightness})
  )
  n27 = TextureCoordinate()
  n28 = Panner(Coordinate=TextureCoordinate() {uTiling=3, vTiling=0}) {speedX=0.2, speedY=0.1}
  n29 = Power(
   Base=OneMinus(None=ComponentMask(None=@n27) {g=true})
   Exp=LinearInterpolate(Alpha=TextureSample(UVs=@n28) {texture=&T_NoisePack_01}.R) {constA=2, constB=6}
  )
  n30 = Saturate(None=Multiply(A=Multiply(A=@n26, B=@n29)) {constB=0.5})
  MaterialFunctionCall(Emissive=Multiply(A=LinearInterpolate(A=@n7, B=@n8.RGB, Alpha=@n30), B=@n30)) {materialFunction=&MF_FX_EmissiveAdjustments}
  n31 = ScalarParameter() {defaultValue=250, parameterName=NearCamFade_Distance}
  n32 = TextureCoordinate()
  n33 = Add(A=Multiply(A=ComponentMask(None=@n32) {g=true}) {constB=2}) {constB=-0.2}
  n34 = DepthFade(FadeDistance=ScalarParameter() {defaultValue=5, parameterName=Depth Fade Distance})
  n35 = Constant() {r=1}
  n36 = Constant() {r=1}
  n37 = Multiply(
   A=Multiply(A=@n8.A, B=Multiply(A=@n33) {constB=2.5})
   B=StaticSwitchParameter(
    True=FeatureLevelSwitch(Default=@n34, ES2=@n35, ES3_1=@n35, SM4_REMOVED=@n34)
    False=QualitySwitch(Default=@n34, Low=@n36, High=@n34, Medium=@n34)
   ) {parameterName=Use_MobilePerf_Settings}
  )
  StaticSwitchParameter(
   True=Multiply(A=MaterialFunctionCall(Distance=@n31) {materialFunction=&Near_Camera_Fade}.Result, B=@n37)
   False=@n37
  ) {parameterName=Use_NearCamFade?}
 ]
}
