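# Unlit translucent Niagara sprite and ribbon material for DiamondSkin weakpoint shards, with a rainbow sheen.
# The rainbow texture is sampled on panning polar UVs offset by the shard normal map, with dynamic parameters randomizing tiling and time.
# The optionally desaturated result is tinted by particle color and shard texture, then passed through emissive adjustments.
# A smoothstep dissolve on the shard green channel, combined with texture and particle alpha, forms opacity.
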Material /Game/Effects/Fort_Effects/GameplayPulginsAssets/DiamondSkin/Materials/M_DiamondSkin_Weakpoint_Shards
graph {
 properties {
  bUsedWithNiagaraRibbons=true
  bUsedWithNiagaraSprites=true
  blendMode=BLEND_Translucent
  pixelDepthOffsetMode=PDOM_Legacy
  shadingModel=MSM_Unlit
  translucencyLightingMode=TLM_Surface
  usedShadingModels=Unlit
 }
 nodes [
  n1 = ParticleColor()
  n2 = VectorParameter() {group=Color, parameterName=Color UV, defaultValue=<rgba 0.5 0.8 -0.01 0>}
  n3 = DynamicParameter() {parameterIndex=1, defaultValue=<rgba 0 0 1 0>, paramNames [UVTilingRandom, TimeOffsetRandom, WPOStrength, TrailingModeON]}
  n4 = Multiply(
   A=MaterialFunctionCall!(UV Coord=TextureCoordinate(), LenghtScale, RadialScale) {materialFunction=/Game/Global/MaterialFunctions/TechArt/MF_PolarCoordinates}.Result
   B=MaterialFunctionCall(X=@n2.R, Y=Add(A=@n3.UVTilingRandom, B=@n2.G)) {materialFunction=&MakeFloat2}.Result
  )
  n5 = Multiply(
   A=MaterialFunctionCall(X=@n2.B, Y=@n2.B) {materialFunction=&MakeFloat2}.Result
   B=Add(A=@n3.TimeOffsetRandom, B=Time())
  )
  n6 = TextureSample() {samplerType=SAMPLERTYPE_Normal, texture=/Game/Effects/Fort_Effects/GameplayPulginsAssets/DiamondSkin/Textures/T_DiamondSkin_Shards_N}
  n7 = ComponentMask(None=@n6.RGB) {b=true, r=true}
  n8 = TextureSampleParameter2D!(UVs=Add(A=Add(A=@n4, B=@n5), B=Multiply(A=@n7) {constB=0.5}) {constB=0.5}, Level, Apply View MipBias) {group=Color, mipValueMode=TMVM_MipLevel, parameterName=Color Tex, texture=&T_Tile_RainbowColor}
  n9 = Desaturation(None=@n8.RGB, Fraction=ScalarParameter() {group=Color, parameterName=Color Tex Desaturation})
  n10 = TextureSample() {samplerType=SAMPLERTYPE_LinearColor, texture=/Game/Effects/Fort_Effects/GameplayPulginsAssets/DiamondSkin/Textures/T_DiamondSkin_Shards}
  MaterialFunctionCall(Emissive=Multiply(A=Multiply(A=@n1.RGB, B=@n9), B=@n10.RGB)) {materialFunction=&MF_FX_EmissiveAdjustments}
  n11 = DynamicParameter() {parameterIndex=2, defaultValue=<rgba 0.1 1 1 0>, paramNames [Smoothness Dissolve, Dissolve, Param3, Param4]}
  n12 = OneMinus(None=@n11.Dissolve)
  n13 = Add(A=@n12, B=@n11.Smoothness Dissolve)
  Multiply(
   A=@n1.A
   B=Multiply(
    A=MaterialFunctionCall(Alpha=@n10.G, Min=@n12, Max=@n13) {materialFunction=&SmoothStep}.Result
    B=@n10.A
   ) {constB=0.2}
  )
 ]
}
